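; Player component Blueprint for the Weapon Mod Station, covering the purchase and removal of weapon mods on a player's weapons.
; Local purchase and remove-mod events, plus affordability checks for both actions, take the mod definition and weapon, with the checks also taking a controller.
; Every override, including BeginPlay, only forwards to its parent, with the purchase events noted as calling into C++, so this Blueprint adds no logic of its own.
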(event EventBeginPlay
  (|CallParentFunction))

(event EventLocalPurchaseWeaponModForWeapon
  (params (WeaponMod :type Fort%20Weapon%20Mod%20Item%20Definition%20Object%20Reference) (Weapon :type Fort%20Weapon%20Object%20Reference))
  (|CallParentFunction WeaponMod Weapon) ; call into c++
)

(event EventLocalPurchaseRemoveMod
  (params (WeaponMod :type Fort%20Weapon%20Mod%20Item%20Definition%20Object%20Reference) (Weapon :type Fort%20Weapon%20Object%20Reference))
  (|CallParentFunction WeaponMod Weapon) ; call into c++
)

(fn CanPlayerAffordModForWeapon
  (params (WeaponMod :type Fort%20Weapon%20Mod%20Item%20Definition%20Object%20Reference) (Weapon :type Fort%20Weapon%20Object%20Reference) (Controller :type Controller%20Object%20Reference))
  (bind _returnvalue (|CallParentFunction WeaponMod Weapon Controller))
  (return _returnvalue))

(fn CanPlayerAffordRemoveModFromWeapon
  (params (WeaponMod :type Fort%20Weapon%20Mod%20Item%20Definition%20Object%20Reference) (Weapon :type Fort%20Weapon%20Object%20Reference) (Controller :type Controller%20Object%20Reference))
  (bind _returnvalue (|CallParentFunction WeaponMod Weapon Controller))
  (return _returnvalue))
